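# Niagara system for the Instant Gravel emote, dressing a vehicle with tire smoke, skid ribbon, debris, crackles, emissive meshes and lights.
# System-level skeletal mesh samples supply position, rotation and velocity to place and launch effects.
# Headlights and taillight combine sprite glows and flares with light renderers; headlights add a component renderer.
# Time-of-day values scale lights, emissives and smoke opacity, while front-end modifiers retune lens flares and smoke.
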NiagaraSystem /BRCosmetics/Animation/Game/MainPlayer/Emotes/InstantGravel/FX/NS_Emote_InstantGravel
dependencies {
 usedDataInterfaces [
  &NiagaraDataInterfaceCurve
  &NiagaraDataInterfaceMeshRendererInfo
  &NiagaraDataInterfaceSkeletalMesh
 ]
 usedDynamicInputs [
  &MaskVectorBySpawnGroup
  &OneMinusFloat
  &FNTimeOfDayScalar
  &RandomRangeInt
  &Subtract_Float
  &VectorLength
 ]
 usedModules [
  &self:SystemSpawnScript.SystemScriptSource.SystemScriptGraph.NiagaraNodeAssignment_2.SetVariables_#1
  &SystemState
  &self:SystemSpawnScript.SystemScriptSource.SystemScriptGraph.NiagaraNodeAssignment_4.SetVariables_#2
  &self:SystemSpawnScript.SystemScriptSource.SystemScriptGraph.NiagaraNodeAssignment_1.SetVariables_#3
  &self:SystemSpawnScript.SystemScriptSource.SystemScriptGraph.NiagaraNodeAssignment_3.SetVariables_#4
  &SampleSkeletalMesh
  &self:SystemSpawnScript.SystemScriptSource.SystemScriptGraph.NiagaraNodeAssignment_0.SetVariables_#5
 ]
}
emitters {
 Crackle {
  emitterUpdateScript [
   &EmitterState()
   &SpawnBurst_Instantaneous(Spawn Count=$System.LightningCrackles, Use Spawn Probability=false, UseLoopCountLimit=false)
  ]
  particleSpawnScript [
   &InitializeParticle(
    Lifetime Max=0.7
    Lifetime Min=0.4
    Lifetime Mode=::NewEnumerator1
    Position=$System.SampleSkeletalMesh.SampledPosition
    Position Mode=::NewEnumerator1
    Position Offset=<0 0 -9>
    Sprite Size=&Multiply_Vector2D_ByFloat(Float=&RandomRangeFloat(Minimum=0.666, Maximum=0.75), Vector2D=<33 66>)
    Sprite Size Mode=::NewEnumerator4
    Sprite UV Mode=::NewEnumerator1
    UsePositionOffset=true
    Write Material Random=true
   )
   &CameraOffset(Camera Offset Amount=11)
   /SetVariables(Particles.SubImageIndex=&ModuloFloat(A=&MakeFloatFromInt(Int=&ReturnParticleIDAttributes()), B=3)) {isSetParameters=true}
  ]
  particleUpdateScript [
   /SetVariables(Particles.SpriteAlignment=&Add_Vector(
     A=&Add_Vector(
      A=&RandomRangeVector(Minimum=<0 0 1>, Maximum=<0 0 1.5>)
      B=&Multiply_VectorByFloat(
       Vector=&NormalizeVector(
        Vector To Normalize=$System.SampleSkeletalMesh.SampledVelocity
        Safe Normalize=true
        Safe Fallback Threshold=0
        Safe Fallback Vector=<0 0 0.001>
       )
       Float=&Lerp_Float(A=-0.6, B=0.1, Alpha=&ReturnNormalizedExecIndex())
      )
     )
     B=&NormalizeVector(
      Vector To Normalize=Linked() {linkedVariable {name=System.SampleSkeletalMesh.SampledPosition, type=NiagaraPosition}}
      Safe Normalize=true
      Safe Fallback Threshold=0
      Safe Fallback Vector=<0 0 0.001>
     )
    )) {isSetParameters=true}
   /SetVariables(Particles.Position=&AddVectorToPosition(
     Position=$System.SampleSkeletalMesh.SampledPosition
     Vector=&Add_Vector(
      A=<0 0 -8>
      B=&Multiply_VectorByFloat(
       Vector=$Particles.SpriteAlignment
       Float=&FloatFromCurve(
        FloatCurve=DataInterface() {propertyValues {Curve {keys [{arriveTangent=4.1785, time=0, value=0}, {arriveTangent=0.0224, leaveTangent=0.0224, time=1, value=1}]}}}
        Scale Curve=&RandomRangeFloat(Minimum=11, Maximum=33)
       )
      )
     )
    )) {isSetParameters=true}
   &ParticleState()
   &ScaleSpriteSize(Uniform Curve Sprite Scale=DataInterface() {propertyValues {Curve {keys [{arriveTangent=1.4704, time=0, value=0.333}, <at 0.25 0.7>, <at 0.6 0.9>, {arriveTangent=-0.0405, leaveTangent=-0.0405, time=1, value=1}]}}})
   &ScaleColor(
    Scale Alpha=&FloatFromCurve(FloatCurve=DataInterface() {propertyValues {Curve {keys [
         <at 0 1>
         {arriveTangent=-0.0405, leaveTangent=-0.0405, time=0.7009, value=0.75}
         {arriveTangent=-8.0089, leaveTangent=-8.0089, time=1, value=0}
        ]}}})
    Scale RGB=&VectorFromFloat(Value=&Lerp_Float(
      A=1
      B=33
      Alpha=&FloatFromCurve(
       FloatCurve=DataInterface() {propertyValues {Curve {keys [
           {arriveTangent=9.0725, interpMode=RCIM_Constant, time=0, value=1}
           <at 0.1 0 RCIM_Constant>
           {arriveTangent=-0.0405, interpMode=RCIM_Constant, leaveTangent=-0.0405, time=1, value=0}
          ]}}}
       CurveIndex=$Particles.Age
       Scale Curve=&RandomRangeFloat(Minimum=0.25, Maximum=1)
      )
     ))
    ScaleA=true
    ScaleRGB=true
   )
  ]
  rendererClasses [&NiagaraSpriteRendererProperties]
  renderers [{rendererClass=&NiagaraSpriteRendererProperties}]
 }
 Debris {
  emitterUpdateScript [
   &EmitterState(
    Inactive Response=::NewEnumerator1
    Life Cycle Mode=::NewEnumerator1
    Loop Behavior=::NewEnumerator1
    Loop Duration Mode=::NewEnumerator1
    UseLoopDelay=false
   )
   &SpawnRate(Spawn Probability=0.2, SpawnRate=$System.SpawnRateDebris, Use Spawn Probability=true)
  ]
  particleSpawnScript [
   /SetVariables(Particles.RandomMass:NiagaraFloat=&RandomRangeFloat(Minimum=0, Maximum=1)) {isSetParameters=true}
   &InitializeParticle(
    Color=<rgba 0.3 0.3 0.3 1>
    Lifetime Max=1
    Lifetime Min=0.1
    Lifetime Mode=::NewEnumerator1
    Mass=&Lerp_Float(A=0.45, B=1.33, Alpha=$Particles.RandomMass)
    Mass Mode=::NewEnumerator1
    Position=$System.SampleSkeletalMesh.SampledPosition
    Position Mode=::NewEnumerator1
    Position Offset=<0 0 -8>
    UsePositionOffset=true
    Write Material Random=true
   )
   /Niagara/Modules/Update/Mass/CalculateSizeAndRotationalInertiaByMass(
    Calculate Mesh Scale=true
    Density by Material Type:ENiagaraMassByVolume=::NewEnumerator1
    Initial Model Dimensions:Vector3f=<33 33 33>
    Module.Manually Enter Density=false
   )
   &AddRotationalVelocity(Rotation Rate=&RandomVector(Vector Scale=&RandomRangeFloat(Minimum=1, Maximum=3)))
   &AddVelocity(Velocity=&Add_Vector(A=&Multiply_VectorByFloat(Vector=$System.SampleSkeletalMesh.SampledVelocity, Float=2), B=<0 0 0>))
   &AddVelocity(
    Cone Axis=&Add_Vector(
     A=<0 0 2>
     B=&NormalizeVector(
      Vector To Normalize=Linked() {linkedVariable {name=System.SampleSkeletalMesh.SampledPosition, type=NiagaraPosition}}
      Safe Normalize=true
      Safe Fallback Threshold=0
      Safe Fallback Vector=<0 0 0.001>
     )
    )
    Module.UseConeFalloff=false
    Module.UseVelDistribution=false
    Override Seed=false
    Velocity Mode=::NewEnumerator2
    Velocity Speed=&Lerp_Float(A=444, B=333, Alpha=$Particles.RandomMass)
   ) {module=AddVelocity001}
   &ApplyInitialForces()
   /SetVariables(Particles.MeshIndex=&MakeIntFromFloat(Float=&Multiply_Float(A=$Particles.RandomMass, B=4), Conversion Mode=::NewEnumerator3)) {isSetParameters=true}
  ]
  particleUpdateScript [
   &ParticleState()
   &ScaleMeshSize(Scale Factor=&VectorFromFloat(Value=&FloatFromCurve(FloatCurve=DataInterface() {propertyValues {Curve {keys [
          {tangentMode=RCTM_User, time=0, value=1}
          {arriveTangent=-0.0405, interpMode=RCIM_Cubic, leaveTangent=-0.0405, tangentMode=RCTM_User, time=0.7009, value=1}
          {arriveTangent=-8.0089, interpMode=RCIM_Cubic, leaveTangent=-8.0089, tangentMode=RCTM_User, time=1, value=0}
         ]}}})))
   &GravityForce()
   &Drag(Use Linear Drag=true, Use Rotational Drag=true)
   &Collision(
    Age Colliding Particles=true
    Analytical Collision Plane Position 1:Vector3f=&AddVectorToPosition(Position=$System.SampleSkeletalMesh.SampledPosition, Vector=<0 0 -10>)
    Analytical Collision Plane Position 2:Vector3f=&AddVectorToPosition(Position=$System.SampleSkeletalMesh.SampledPosition, Vector=<0 0 -10>)
    CPU Collision Type=::NewEnumerator3
    Collision Enabled=&SetBoolByFloatComparison(A=$Particles.Age, B=0.2)
    KillOnCollision=false
    Module.Control Rest State Via Surface Normal=false
    Module.EnableMaxCollisionCount=false
    Module.Randomize Collision Normal Vector=false
   )
   &SolveForcesAndVelocity(Clamp Velocity=false, Limit Acceleration=false, Manually Enable Rotational Solver=true)
  ]
  rendererClasses [&NiagaraMeshRendererProperties]
  renderers [{rendererClass=&NiagaraMeshRendererProperties, rendererIndex=0}]
 }
 Headlights {
  emitterSpawnScript [/SetVariables(
    Emitter.BeamSize:Vector2f=<133 222>
    Emitter.CapsuleGlowLength=44
    Emitter.CapsuleGlowSize:Vector2f=&Vector2DFromFloat(Value=22)
    Emitter.CapsuleGlowStemLength=33
    Emitter.CapsuleGlowStemSize:Vector2f=&Vector2DFromFloat(Value=16)
    Emitter.FlareCameraOffset=22
    Emitter.FlareSize:Vector2f=&Vector2DFromFloat(Value=133)
    Emitter.GlowCameraOffset=11
    Emitter.GlowSize:Vector2f=&Vector2DFromFloat(Value=55)
    Emitter.LensFlareElement1Offset:NiagaraFloat=&FN_FrontEndFloatModifier(Gameplay=0.333, Front End=0.166)
    Emitter.LensFlareElement1Size:Vector2f=&Vector2DFromFloat(Value=&FN_FrontEndFloatModifier(Gameplay=0.133, Front End=0.17))
    Emitter.LensFlareElement2Offset:NiagaraFloat=&FN_FrontEndFloatModifier(Gameplay=0.533, Front End=0.266)
    Emitter.LensFlareElement2Size:Vector2f=&Vector2DFromFloat(Value=&FN_FrontEndFloatModifier(Gameplay=0.2, Front End=0.26))
    Emitter.LensFlareElement3Offset:NiagaraFloat=&FN_FrontEndFloatModifier(Gameplay=0.633, Front End=0.316)
    Emitter.LensFlareElement3Size:Vector2f=&Vector2DFromFloat(Value=&FN_FrontEndFloatModifier(Gameplay=0.233, Front End=0.3))
    Emitter.LensFlareRainbowOffset:NiagaraFloat=&FN_FrontEndFloatModifier(Gameplay=0.1, Front End=0.033)
    Emitter.LensFlareRainbowSize:Vector2f=&Vector2DFromFloat(Value=&FN_FrontEndFloatModifier(Gameplay=0.333, Front End=0.466))
   ) {isSetParameters=true}]
  emitterUpdateScript [
   &EmitterState(
    Inactive Response=::NewEnumerator1
    Life Cycle Mode=::NewEnumerator1
    Loop Behavior=::NewEnumerator1
    Loop Duration Mode=::NewEnumerator1
    UseLoopDelay=false
   )
   &SpawnBurst_Instantaneous(Use Spawn Probability=false, UseLoopCountLimit=false)
   /SetVariables(
    Emitter.CapsuleGlowAxis:Vector3f=&TransformVector(
     Vector=&RotateVectorByQuaternion(Vector To Rotate=<0 0 1>, Quaternion=$System.SampleSkeletalMesh001.SampledRotation)
     Destination Space=::World
    )
    Emitter.HeadlightColor:LinearColor=&MultiplyLinearColor(
     Linear Color=<rgba 1 0.9417 0.3 1>
     Scale Factor=&MakeVector4_FromFloat(Float=$System.TimeOfDayHeadlightIntensity)
    )
    Emitter.HeadlightDirection:Vector3f=&RotateVectorByQuaternion(Vector To Rotate=<0 1 0>, Quaternion=$System.SampleSkeletalMesh001.SampledRotation)
   ) {isSetParameters=true}
   /SetVariables(
    Emitter.BeamAlignment:Vector3f=&Multiply_VectorByFloat(Vector=$Emitter.HeadlightDirection, Float=-1)
    Emitter.BeamColor:LinearColor=&MultiplyLinearColor(Linear Color=$Emitter.HeadlightColor, Scale Factor=&MakeVector4_FromFloat(Float=0.333))
    Emitter.CapsuleGlowColor:LinearColor=&MultiplyLinearColor(Linear Color=$Emitter.HeadlightColor, Scale Factor=&MakeVector4_FromFloat(Float=1))
    Emitter.CapsuleGlowDynamic:Vector4f=&MakeFloat4FromVectorAndFloat("Vector (RGB)"=$Emitter.CapsuleGlowAxis, "Float (Alpha)"=$Emitter.CapsuleGlowLength)
    Emitter.CapsuleGlowStemColor:LinearColor=&MultiplyLinearColor(Linear Color=$Emitter.HeadlightColor, Scale Factor=&MakeVector4_FromFloat(Float=0.666))
    Emitter.CapsuleGlowStemDynamic:Vector4f=&MakeFloat4FromVectorAndFloat("Vector (RGB)"=$Emitter.CapsuleGlowAxis, "Float (Alpha)"=$Emitter.CapsuleGlowStemLength)
    Emitter.FlareColor:LinearColor=&MultiplyLinearColor(Linear Color=$Emitter.HeadlightColor, Scale Factor=&MakeVector4_FromFloat(Float=1))
    Emitter.GlowColor:LinearColor=&MultiplyLinearColor(Linear Color=$Emitter.HeadlightColor, Scale Factor=&MakeVector4_FromFloat(Float=0.333))
    Emitter.LensFlareColor:LinearColor=&MultiplyLinearColor(Linear Color=$Emitter.HeadlightColor, Scale Factor=&MakeVector4_FromFloat(Float=0.4))
    Emitter.LensFlareElement1Dynamic:Vector4f=&MakeFloat4FromVectorAndFloat("Vector (RGB)"=$Emitter.HeadlightDirection, "Float (Alpha)"=$Emitter.Initial.LensFlareElement1Offset)
    Emitter.LensFlareElement2Dynamic:Vector4f=&MakeFloat4FromVectorAndFloat("Vector (RGB)"=$Emitter.HeadlightDirection, "Float (Alpha)"=$Emitter.Initial.LensFlareElement2Offset)
    Emitter.LensFlareElement3Dynamic:Vector4f=&MakeFloat4FromVectorAndFloat("Vector (RGB)"=$Emitter.HeadlightDirection, "Float (Alpha)"=$Emitter.Initial.LensFlareElement3Offset)
    Emitter.LensFlareRainbowDynamic:Vector4f=&MakeFloat4FromVectorAndFloat("Vector (RGB)"=$Emitter.HeadlightDirection, "Float (Alpha)"=$Emitter.Initial.LensFlareRainbowOffset)
    Emitter.LightColor:LinearColor=&MultiplyLinearColor(Linear Color=$Emitter.HeadlightColor, Scale Factor=&MakeVector4_FromFloat(Float=3))
   ) {isSetParameters=true}
  ]
  particleSpawnScript [&InitializeParticle(
    Color=<rgba 1 0.9417 0.3 1>
    Position Mode=::NewEnumerator0
    Sprite Size Mode=::NewEnumerator3
    Uniform Sprite Size=0.2
    Write Material Random=true
   )]
  particleUpdateScript [
   /SetVariables(
    Particles.CapsuleGlowStemPosition:NiagaraPosition=&AddVectorToPosition(
     Position=$System.SampleSkeletalMesh001.SampledPosition
     Vector=&RotateVectorByQuaternion(Vector To Rotate=<0 4.5 -27.33>, Quaternion=$System.SampleSkeletalMesh001.SampledRotation)
    )
    Particles.Color=&MultiplyLinearColor(
     Linear Color=$Particles.Initial.Color
     Scale Factor=&MakeVector4_FromFloat(Float=$System.TimeOfDayHeadlightIntensity)
    )
    Particles.Position=&AddVectorToPosition(
     Position=$System.SampleSkeletalMesh001.SampledPosition
     Vector=&RotateVectorByQuaternion(Vector To Rotate=<0 13 -2.666>, Quaternion=$System.SampleSkeletalMesh001.SampledRotation)
    )
    Particles.SpotlightIntensity:NiagaraFloat=&Multiply_Float(A=20, B=$System.TimeOfDayHeadlightIntensity)
    Particles.SpotlightRotation:Quat4f=&MultiplyQuaternion(
     Quaternion A=$System.SampleSkeletalMesh001.SampledRotation
     Quaternion B=&MakeQuatFromAxisAngle(Axis=<0 0 1>, AngleInDegrees=90)
    )
   ) {isSetParameters=true, module=SetVariables_#6}
   &ParticleState(Kill Particles When Lifetime Has Elapsed=false, Let Infinitely Lived Particles Die When Emitter Deactivates=true)
  ]
  rendererClasses [
   &NiagaraSpriteRendererProperties
   &NiagaraComponentRendererProperties
   &NiagaraLightRendererProperties
  ]
  renderers [
   {rendererClass=&NiagaraSpriteRendererProperties}
   {rendererClass=&NiagaraSpriteRendererProperties, rendererIndex=1}
   {rendererClass=&NiagaraSpriteRendererProperties, rendererIndex=2}
   {rendererClass=&NiagaraSpriteRendererProperties, rendererIndex=3}
   {rendererClass=&NiagaraSpriteRendererProperties, rendererIndex=4}
   {rendererClass=&NiagaraComponentRendererProperties, rendererIndex=5}
   {rendererClass=&NiagaraSpriteRendererProperties, rendererIndex=6}
   {rendererClass=&NiagaraSpriteRendererProperties, rendererIndex=7}
   {rendererClass=&NiagaraSpriteRendererProperties, rendererIndex=8}
   {rendererClass=&NiagaraSpriteRendererProperties, rendererIndex=9}
   {rendererClass=&NiagaraLightRendererProperties, rendererIndex=10}
  ]
 }
 MeshEmissives {
  emitterUpdateScript [
   &EmitterState(
    Inactive Response=::NewEnumerator1
    Life Cycle Mode=::NewEnumerator1
    Loop Behavior=::NewEnumerator1
    Loop Duration Mode=::NewEnumerator1
    UseLoopDelay=false
   )
   &SpawnBurst_Instantaneous(Use Spawn Probability=false, UseLoopCountLimit=false)
  ]
  particleSpawnScript [&InitializeParticle(
    Color=<rgba 1.5 1.5 1.5 1>
    Lifetime=5
    Mesh Scale=$Engine.Owner.Scale
    Mesh Scale Mode=::NewEnumerator4
    UsePositionOffset=false
    Write Material Random=true
   )]
  particleUpdateScript [
   &ParticleState(Kill Particles When Lifetime Has Elapsed=false)
   &ScaleColor(Scale RGB=&VectorFromFloat(Value=$System.TimeOfDayHeadlightIntensity), ScaleA=true, ScaleRGB=true)
  ]
  rendererClasses [&NiagaraMeshRendererProperties]
  renderers [
   {rendererClass=&NiagaraMeshRendererProperties, rendererIndex=0}
   {rendererClass=&NiagaraMeshRendererProperties, rendererIndex=1}
  ]
 }
 SkidRibbon {
  emitterUpdateScript [
   &EmitterState(
    Inactive Response=::NewEnumerator1
    Life Cycle Mode=::NewEnumerator1
    Loop Behavior=::NewEnumerator1
    Loop Duration Mode=::NewEnumerator1
    UseLoopDelay=false
   )
   &SpawnRate(SpawnRate=11, Use Spawn Probability=false)
  ]
  particleSpawnScript [&InitializeParticle(
    Color=&MultiplyLinearColor(
     Linear Color=<rgba 0.1 0.145 0.2 1>
     Scale Factor=&MakeFloat4FromVectorAndFloat(
      "Vector (RGB)"=<1 1 1>
      "Float (Alpha)"=&Multiply_Float(
       A=$System.Opacity
       B=&Waveform("Waveform [1]"=::NewEnumerator0, "[1] Phase Scale"=2.6, "[1] Frequency"=1, "[1] Amplitude Min/Max"=<0.5 1>)
      )
     )
    )
    Lifetime=5
    Position=&TransformPosition(Position=$System.SampleSkeletalMesh.SampledPosition, Source Space=::Local, Destination Space=::Simulation)
    Position Mode=::NewEnumerator1
    Position Offset=&Add_Vector(
     A=<0 0 -8>
     B=&Multiply_VectorByFloat(
      Vector=&NormalizeVector(
       Vector To Normalize=&Multiply_Vector(A=Linked() {linkedVariable {name=System.SampleSkeletalMesh.SampledPosition, type=NiagaraPosition}}, B=<1 1 0>)
       Safe Normalize=true
       Safe Fallback Threshold=0
       Safe Fallback Vector=<0 0 0.001>
      )
      Float=&Waveform(Global Amplitude Scale=6, "Waveform [1]"=::NewEnumerator0, "[1] Frequency"=0.618, "[1] Amplitude Min/Max"=<-1 1>)
     )
    )
    Ribbon Facing Mode=::NewEnumerator1
    Ribbon Width=7
    Ribbon Width Mode=::NewEnumerator1
    UsePositionOffset=true
    Write Material Random=true
   )]
  particleUpdateScript [
   &ParticleState()
   &ScaleColor(
    Scale Alpha=&FloatFromCurve(
     FloatCurve=DataInterface() {propertyValues {Curve {keys [<at 0 1>,<at 1 0>]}}}
     Scale Curve=&FloatFromCurve(FloatCurve=DataInterface() {propertyValues {Curve {keys [<at 0 0>,<at 0.1 1>]}}}, CurveIndex=$Particles.Age)
    )
    ScaleA=true
    ScaleRGB=false
   )
  ]
  rendererClasses [&NiagaraRibbonRendererProperties]
  renderers [{rendererClass=&NiagaraRibbonRendererProperties}]
 }
 SmokeTire {
  emitterUpdateScript [
   &EmitterState(
    Inactive Response=::NewEnumerator1
    Life Cycle Mode=::NewEnumerator1
    Loop Behavior=::NewEnumerator1
    Loop Duration Mode=::NewEnumerator1
    UseLoopDelay=false
   )
   &SpawnBurst_Instantaneous(Spawn Count=6, Use Spawn Probability=false, UseLoopCountLimit=false)
  ]
  particleSpawnScript [
   &InitializeParticle(
    Color=&LerpLinearColors(
     StartColor=<rgba 1 1 1 1>
     EndColor=<rgba 0.585 0.6675 0.75 1>
     LerpFactor=&FN_FrontEndFloatModifier(Gameplay=0, Front End=1)
    )
    Sprite Rotation Mode=::NewEnumerator2
    Sprite Size Mode=::NewEnumerator1
    Uniform Sprite Size Max=22
    Uniform Sprite Size Min=11
    UsePositionOffset=false
    Write Material Random=true
   )
   &CameraOffset(Camera Offset Amount=11)
  ]
  particleUpdateScript [
   &ShapeLocation(
    Additional Quaternion Rotation=true
    Additional Yaw / Pitch / Roll=<0 0.25 0>
    MaskSpawnGroup=false
    Offset Mode=::NewEnumerator0
    Override Seed=false
    Ring / Disc Distribution Mode:ENiagaraSphereDistributionMode=::NewEnumerator1
    Ring Radius=7
    Rotation Mode=::NewEnumerator3
    Rotation Quaternion=$System.SampleSkeletalMesh.SampledRotation
    Shape Origin=$System.SampleSkeletalMesh.SampledPosition
    Shape Primitive=::NewEnumerator4
    U Position=&Frac_Float(Float=&Add_Float(A=&Multiply_Float(A=$Engine.Time, B=-1.666), B=&ReturnNormalizedExecIndex()))
    UseOwnerScale=true
   )
   &ParticleState(Kill Particles When Lifetime Has Elapsed=false)
   &SpriteRotationRate(Rotation Rate=360)
   &DynamicMaterialParameters(
    Index 0 Param 1=&Frac_Float(Float=&Add_Float(A=&ReturnNormalizedExecIndex(), B=&Multiply_Float(A=$Engine.Time, B=0.5)))
    Index 0 Param 4=1
    Module.Index0Param0Write=true
    Module.Index0Param1Write=true
    Module.Index0Param2Write=true
    Module.Index0Param3Write=true
   )
   &ScaleColor(Scale Alpha=$System.TimeOfDaySmokeOpacity, Scale RGB=$System.SmokeLightingColor, ScaleA=true, ScaleRGB=true)
  ]
  rendererClasses [&NiagaraSpriteRendererProperties]
  renderers [{rendererClass=&NiagaraSpriteRendererProperties}]
 }
 SmokeTrailing {
  emitterUpdateScript [
   &EmitterState(
    Inactive Response=::NewEnumerator1
    Life Cycle Mode=::NewEnumerator1
    Loop Behavior=::NewEnumerator1
    Loop Delay=1
    Loop Duration Mode=::NewEnumerator1
    UseLoopDelay=true
   )
   &SpawnRate(SpawnRate=$System.SpawnRateSmokeTrailing, Use Spawn Probability=false)
  ]
  particleSpawnScript [
   &InitializeParticle(
    Color=&LerpLinearColors(
     StartColor=<rgba 1 1 1 1>
     EndColor=<rgba 0.585 0.6675 0.75 1>
     LerpFactor=&FN_FrontEndFloatModifier(Gameplay=0, Front End=1)
    )
    Lifetime Max=&FN_FrontEndFloatModifier(Gameplay=1.62, Front End=2.62)
    Lifetime Min:NiagaraFloat=&FN_FrontEndFloatModifier(Gameplay=1.26, Front End=2.26)
    Lifetime Mode=::NewEnumerator1
    Position=$System.SampleSkeletalMesh.SampledPosition
    Position Mode=::NewEnumerator1
    Sprite Rotation Mode=::NewEnumerator2
    Sprite Size Mode=::NewEnumerator1
    Sprite UV Mode=::NewEnumerator3
    Uniform Sprite Size Max=77
    Uniform Sprite Size Min=44
    UsePositionOffset=false
    Write Material Random=true
   )
   &AddVelocity(Velocity=$System.SampleSkeletalMesh.SampledVelocity, Velocity Speed Scale=2)
  ]
  particleUpdateScript [
   &ParticleState()
   &Drag(Drag=7, Use Linear Drag=true, Use Rotational Drag=true)
   &GravityForce(Gravity=&Multiply_VectorByFloat(
     Vector=&RandomRangeVector(Minimum=<0 0 -66>, Maximum=<1 1 222>)
     Float=&FN_FrontEndFloatModifier(Gameplay=1, Front End=0.5)
    ))
   &PointForce(
    Force Falloff Distance=0
    Force Strength=&FloatFromCurve(FloatCurve=DataInterface() {propertyValues {Curve {keys [<at 0 1>,<at 1 0>]}}}, CurveIndex=$Particles.Age, Scale Curve=333)
    Module.Use Falloff=false
    Module.UseLinearForce=false
   )
   &SolveForcesAndVelocity(Clamp Velocity=true, Limit Acceleration=false, Speed Limit=400)
   &ScaleSpriteSize(Uniform Curve Sprite Scale=DataInterface() {propertyValues {Curve {keys [{arriveTangent=1.4704, time=0, value=0.333}, <at 0.2 0.7>, <at 0.6 0.9>, {arriveTangent=-0.0405, leaveTangent=-0.0405, time=1, value=1}]}}})
   &DynamicMaterialParameters(
    Index 0 Param 1=&Add_Float(
     A=$Particles.RandomFloat
     B=&Multiply_Float(A=$Particles.Age, B=&FN_FrontEndFloatModifier(Gameplay=0.333, Front End=0.166))
    )
    Index 0 Param 4=1
    Index 2 Param 1=&FloatFromCurve(FloatCurve=DataInterface() {propertyValues {Curve {keys [
         {arriveTangent=4.1785, interpMode=RCIM_Cubic, leaveTangent=4.1785, tangentMode=RCTM_User, time=0, value=0}
         {arriveTangent=1.1179, interpMode=RCIM_Cubic, leaveTangent=1.1179, tangentMode=RCTM_User, time=0.2042, value=0.6228}
         {arriveTangent=0.0224, interpMode=RCIM_Cubic, leaveTangent=0.0224, tangentMode=RCTM_User, time=1, value=1}
        ]}}})
    Index 2 Param 3=&FloatFromCurve(
     FloatCurve=DataInterface() {propertyValues {Curve {keys [
         {interpMode=RCIM_Cubic, tangentMode=RCTM_Break, time=0, value=0}
         {arriveTangent=2.4467, interpMode=RCIM_Cubic, tangentMode=RCTM_Break, time=1, value=1}
        ]}}}
     Scale Curve=22
    )
    Index 2 Param 4=0.6
    Module.Param0WriteEnabled=true
    Module.Param0WriteEnabled1=true
    Module.Param0WriteEnabled2=true
    Module.Param1WriteEnabled=false
    Module.Param1WriteEnabled1=true
    Module.Param1WriteEnabled2=true
    Module.Param2WriteEnabled=true
    Module.Param2WriteEnabled1=true
    Module.Param2WriteEnabled2=true
    Module.Param3WriteEnabled=true
    Module.Param3WriteEnabled1=true
    Module.Param3WriteEnabled2=true
    Write Parameter Index 1=true
    Write Parameter Index 2=true
   )
   &ScaleColor(
    Scale Alpha=$System.TimeOfDaySmokeOpacity
    Scale RGB=&Lerp_Vector(
     A=$System.SmokeLightingColor
     B=<1 1 1>
     Alpha=&ClampFloat(Float=&Divide_Float(
       A=&DistanceBetweenPositions(A=$Particles.Position, B=$System.SampleSkeletalMesh.SampledPosition)
       B=66
      ))
    )
    ScaleA=true
    ScaleRGB=true
   )
   &CameraOffset(Camera Offset Amount=&FloatFromCurve(
     FloatCurve=DataInterface() {propertyValues {Curve {keys [{tangentMode=RCTM_User, time=0, value=0}, {tangentMode=RCTM_User, time=1, value=1}]}}}
     Scale Curve=&FN_FrontEndFloatModifier(Gameplay=66, Front End=111)
    ))
  ]
  rendererClasses [&NiagaraSpriteRendererProperties]
  renderers [
   {rendererClass=&NiagaraSpriteRendererProperties}
   {rendererClass=&NiagaraSpriteRendererProperties, rendererIndex=1}
  ]
 }
 Tailight {
  emitterSpawnScript [/SetVariables(
    Emitter.FlareTaillightSize:Vector2f=&Vector2DFromFloat(Value=66)
    Emitter.GlowTaillightCameraOffset=22
    Emitter.GlowTaillightSize:Vector2f=&Vector2DFromFloat(Value=44)
   ) {isSetParameters=true}]
  emitterUpdateScript [
   &EmitterState(
    Inactive Response=::NewEnumerator1
    Life Cycle Mode=::NewEnumerator1
    Loop Behavior=::NewEnumerator1
    Loop Duration Mode=::NewEnumerator1
    UseLoopDelay=false
   )
   &SpawnBurst_Instantaneous(Use Spawn Probability=false, UseLoopCountLimit=false)
   /SetVariables(
    Emitter.FlareTaillightFacing:Vector3f=&RotateVectorByQuaternion(Vector To Rotate=<0 0 1>, Quaternion=$System.SampleSkeletalMesh002.SampledRotation)
    Emitter.TaillightColor:LinearColor=&MultiplyLinearColor(
     Linear Color=<rgba 3 0 0 1>
     Scale Factor=&MakeVector4_FromFloat(Float=$System.TimeOfDayTaillightIntensity)
    )
   ) {isSetParameters=true}
   /SetVariables(
    Emitter.FlareTaillightColor:LinearColor=&MultiplyLinearColor(Linear Color=$Emitter.TaillightColor, Scale Factor=&MakeVector4_FromFloat(Float=1))
    Emitter.GlowTaillightColor:LinearColor=&MultiplyLinearColor(Linear Color=$Emitter.TaillightColor, Scale Factor=&MakeVector4_FromFloat(Float=0.333))
   ) {isSetParameters=true}
  ]
  particleSpawnScript [&InitializeParticle(
    Color=<rgba 3 0 0 1>
    Mesh Scale Mode=::NewEnumerator3
    Position Mode=::NewEnumerator0
    Sprite Size Mode=::NewEnumerator3
    Uniform Sprite Size=111
    Write Material Random=true
   )]
  particleUpdateScript [
   /SetVariables(Particles.Position=&AddVectorToPosition(
     Position=$System.SampleSkeletalMesh.SampledPosition
     Vector=&RotateVectorByQuaternion(Vector To Rotate=<0 0 11>, Quaternion=$System.SampleSkeletalMesh002.SampledRotation)
    )) {isSetParameters=true, module=SetVariables_#6}
   &ParticleState(Kill Particles When Lifetime Has Elapsed=false, Let Infinitely Lived Particles Die When Emitter Deactivates=true)
  ]
  rendererClasses [&NiagaraLightRendererProperties, &NiagaraSpriteRendererProperties]
  renderers [
   {rendererClass=&NiagaraLightRendererProperties, rendererIndex=0}
   {rendererClass=&NiagaraSpriteRendererProperties, rendererIndex=1}
   {rendererClass=&NiagaraSpriteRendererProperties, rendererIndex=2}
  ]
 }
}
summary {
 systemName=NS_Emote_InstantGravel
 emitters [
  {emitterName=SmokeTrailing, rendererClasses [&NiagaraSpriteRendererProperties]}
  {emitterName=SmokeTire, rendererClasses [&NiagaraSpriteRendererProperties]}
  {emitterName=SkidRibbon, rendererClasses [&NiagaraRibbonRendererProperties]}
  {emitterName=Debris, rendererClasses [&NiagaraMeshRendererProperties]}
  {
   emitterName=Headlights
   rendererClasses [
    &NiagaraSpriteRendererProperties
    &NiagaraComponentRendererProperties
    &NiagaraLightRendererProperties
   ]
  }
  {emitterName=Tailight, rendererClasses [&NiagaraLightRendererProperties, &NiagaraSpriteRendererProperties]}
  {emitterName=Crackle, rendererClasses [&NiagaraSpriteRendererProperties]}
  {emitterName=MeshEmissives, rendererClasses [&NiagaraMeshRendererProperties]}
 ]
 userVariables [
  User.bIsFrontEnd {type=NiagaraBool}
  User.bIsFrontEndPreview {type=NiagaraBool}
 ]
}
